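; Player state component that grants a player every missing boon in BoonList upon reaching CherrySmoke hero rank S+.
; CheckIfPawnValid retries on a capped timer for a pawn, then listens for OnHeroRankChanged through the event router.
; On the owner's S+, it stops listening; authority spawns pickups targeted at the pawn, or gives items directly under the IcicleLarva tag.
; A cheat event duplicates the grant without rank checks.
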(event Custom|Cheat_GrantAllBoonsOnHeroRankSPlus
  (bind _as_fort_player_pawn_athena (Variables|Default|GetAs%20Fort%20Player%20Pawn%20Athena))
  (bind _returnvalue (Fortnite|GetFortPlayerControllerfromActor _as_fort_player_pawn_athena))
  ;[0 0.1164 0.8594 0.5] GRANT ALL BOONS ON S+
  (bind _asfort_player_state (Utilities|Casting|CastToFortPlayerState (Components|GetOwner)))
  (bind _asfort_player_pawn_athena (Utilities|Casting|CastToFortPlayerPawnAthena (PlayerState|GetPawn _asfort_player_state)))
  (Variables|Default|SetAs%20Fort%20Player%20Pawn%20Athena _asfort_player_pawn_athena)
  (Utilities|FlowControl|Sequence
    (:then_0
      (Networking|SwitchHasAuthority
        (:Authority
          (for _array_element (Variables|Default|GetBoonList)
            (bind _returnvalue_2 (InventoryManagement|GetItemCountforItemDefinition _returnvalue _array_element))
            ;;[0.15 0.008483 0.03963 0.5] Checking if the player has the IcicleLarva tag and if so give the boon directly instead of spawning a pickup
            (if (== _returnvalue_2 0)
              (if (GameplayTags|HasMatchingGameplayTag (Variables|Default|GetAs%20Fort%20Player%20Pawn%20Athena) "(TagName=\"Granted.IcicleLarva.Equipped.Special\")")
                (Inventory|GiveItemtoInventoryOwner _returnvalue _array_element)
                (else
                  (bind _returnvalue_3 (Inventory|SpawnPickupInWorldWithClassAndItemEntry (Loot|CreateItemEntry _array_element) "/Script/FortniteGame.FortPickupAthena" (+ (Transformation|GetActorLocation (Variables|Default|GetAs%20Fort%20Player%20Pawn%20Athena)) (Variables|Default|GetBoon%20Spawn%20Offset)) "0.000000,0.000000,1.000000" 0 true false false "Player" "TossedByPlayer" 0 false))
                  (Pickup|BlueprintSetPickupTarget _returnvalue_3 (Variables|Default|GetAs%20Fort%20Player%20Pawn%20Athena) true))))))))))

(event Custom|GrantAllBoonsOnHeroRankSPlus
  (params (NewRank :type Gameplay%20Tag%20Structure%20%28by%20ref%29) (FortPlayerState :type Fort%20Player%20State%20Object%20Reference))
  (bind _as_fort_player_pawn_athena (Variables|Default|GetAs%20Fort%20Player%20Pawn%20Athena))
  (bind _returnvalue (Fortnite|GetFortPlayerControllerfromActor _as_fort_player_pawn_athena))
  ;[0 0.1164 0.8594 0.5] GRANT ALL BOONS ON S+
  (if (== NewRank "(TagName=\"CherrySmoke.HeroRank.S+\")")
    (bind _asfort_player_state (Utilities|Casting|CastToFortPlayerState (Components|GetOwner)))
    (bind _asfort_player_pawn_athena (Utilities|Casting|CastToFortPlayerPawnAthena (PlayerState|GetPawn _asfort_player_state)))
    (if (== (PlayerState|GetPawn _asfort_player_state) (PlayerState|GetPawn FortPlayerState))
      (Variables|Default|SetAs%20Fort%20Player%20Pawn%20Athena _asfort_player_pawn_athena)
      (GameplayEventRouter|StopListeningtoEvent (Variables|Default|GetS%20Plus%20Async%20Task%20Ref))
      (Utilities|FlowControl|Sequence
        (:then_0
          (Networking|SwitchHasAuthority
            (:Authority
              (for _array_element (Variables|Default|GetBoonList)
                (bind _returnvalue_2 (InventoryManagement|GetItemCountforItemDefinition _returnvalue _array_element))
                ;;[0.15 0.008483 0.03963 0.5] Checking if the player has the IcicleLarva tag and if so give the boon directly instead of spawning a pickup
                (if (== _returnvalue_2 0)
                  (if (GameplayTags|HasMatchingGameplayTag (Variables|Default|GetAs%20Fort%20Player%20Pawn%20Athena) "(TagName=\"Granted.IcicleLarva.Equipped.Special\")")
                    (Inventory|GiveItemtoInventoryOwner _returnvalue _array_element)
                    (else
                      (bind _returnvalue_3 (Inventory|SpawnPickupInWorldWithClassAndItemEntry (Loot|CreateItemEntry _array_element) "/Script/FortniteGame.FortPickupAthena" (+ (Transformation|GetActorLocation (Variables|Default|GetAs%20Fort%20Player%20Pawn%20Athena)) (Variables|Default|GetBoon%20Spawn%20Offset)) "0.000000,0.000000,1.000000" 0 true false false "Player" "TossedByPlayer" 0 false))
                      (Pickup|BlueprintSetPickupTarget _returnvalue_3 (Variables|Default|GetAs%20Fort%20Player%20Pawn%20Athena) true))))))))))))

(event Custom|CheckIfPawnValid
  (bind _asfort_player_state (Utilities|Casting|CastToFortPlayerState (Components|GetOwner)))
  (bind _asfort_player_pawn_athena (Utilities|Casting|CastToFortPlayerPawnAthena (PlayerState|GetPawn _asfort_player_state))
    (:then
      (bind _returnvalue (GameplayEventRouter|GetEventRouter "/Script/FortniteGame.FortGameState" _asfort_player_pawn_athena))
      (bind (_asynctaskproxy _payload) (GameplayEventRouter|StartListeningToEvent _returnvalue "/Script/CherrySmokeHeroRankingSystemRuntime.OnHeroRankChanged" _asfort_player_pawn_athena)
        (:then
          (Variables|Default|SetS%20Plus%20Async%20Task%20Ref _asynctaskproxy))
        (:OnEventReceived
          (bind (_payload_ranktag _payload_player) (Utilities|Struct|BreakOnHeroRankChanged _payload))
          (CallFunction|GrantAllBoonsonHeroRankSPlus _payload_ranktag _payload_player))))
    (:CastFailed
      (if (< (Variables|Default|GetByteTotalChecksForValidPawn) 6)
        (bind _returnvalue_2 (Utilities|Time|SetTimerbyEvent (delegate Custom|CheckIfPawnValid) 5.0))
        (Variables|Default|SetCheckIfPawnValidTimerHandle _returnvalue_2)
        (Variables|Default|SetByteTotalChecksForValidPawn (+ (Variables|Default|GetByteTotalChecksForValidPawn) 1))
        (else
          (Utilities|Time|ClearandInvalidateTimerbyHandle (Variables|Default|GetCheckIfPawnValidTimerHandle)))))))

(event EventBeginPlay
  (|CallParentFunction))

(event EventTick
  (params (DeltaSeconds :type Float%20%28single-precision%29)))
